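# Masked Niagara sprite material for crackle dot particles that fakes a lit sphere on a flat sprite.
# An engine sphere normal map is dotted with a normalized light vector from dynamic parameters, then shaped by that vector's alignment to the camera.
# A radial sphere-mask EdgeGlow term feeds the backlit side of the blend.
# The result, scaled by HighlightOpacity, blends particle color toward a dynamic highlight color through MF_FX_EmissiveAdjustments.
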Material /Game/Effects/Fort_Effects/Materials/Toon/M_Crackle_Dot_Particle
graph {
 properties {bUsedWithNiagaraSprites=true, blendMode=BLEND_Masked, pixelDepthOffsetMode=PDOM_Legacy, usedShadingModels=Default Lit}
 nodes [
  Constant()
  n1 = DynamicParameter() {defaultValue=<rgba 0.5 0 1 2>, paramNames [R, G, B, HighlightOpacity]}
  n2 = DynamicParameter() {parameterIndex=1, defaultValue=<rgba 1 0 0.5 2>, paramNames [X, Y, Z, LightingPower]}
  n3 = Normalize(VectorInput=MaterialFunctionCall(R=@n2.X, G=@n2.X, B=@n2.X) {materialFunction=&AppendMany}.RGB)
  n4 = Saturate(None=DotProduct(A=@n3, B=CameraVectorWS()))
  n5 = TextureSample() {samplerType=SAMPLERTYPE_Normal, texture=/Engine/Functions/Engine_MaterialFunctions02/ExampleContent/Textures/SphereRenderNormalsMap}
  n6 = DotProduct(A=Transform(None=@n5.RGB), B=@n3)
  n7 = Multiply(
   A=LinearInterpolate(A=Multiply(A=@n4) {constB=0.2}, Alpha=Power(Base=@n6)) {constA=0.1}
   B=LinearInterpolate(A=NamedRerouteUsage(), Alpha=@n4) {constB=0.5}
  )
  MaterialFunctionCall(Emissive=LinearInterpolate(
    A=ParticleColor().RGB
    B=MaterialFunctionCall(R=@n1.R, G=@n1.R, B=@n1.R) {materialFunction=&AppendMany}.RGB
    Alpha=Multiply(A=@n1.HighlightOpacity, B=@n7)
   )) {materialFunction=&MF_FX_EmissiveAdjustments}
  Multiply(A=VectorParameter() {parameterName=NormalMult, defaultValue=<rgba 1 1 2 1>}.RGB, B=@n5.RGB)
  n8 = TextureCoordinate()
  n9 = Constant() {r=0.5}
  n10 = Constant() {r=0.45}
  n11 = OneMinus(None=SphereMask(A=@n8, B=@n9, Radius=@n10) {hardnessPercent=0})
  NamedRerouteDeclaration(None=Multiply(A=SphereMask(A=@n8, B=@n9, Radius=@n10), B=Power(Base=@n11) {constExponent=8})) {name=EdgeGlow, nodeColor=<rgba 0 0.4844 1 1>}
 ]
}
